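; Clone machine gameplay actor for Athena that plays the sound and visual effects for its activation, player respawn and destruction.
; Activation and respawn events start sounds at the Location child actor and set timers from replicated server end times to fade them out.
; When a player spawns, the server attaches a short-lived VictoryDrone and activates invincibility visuals, while clients run a local respawn gameplay cue.
; On death, a blast sound plays where cosmetics are allowed.
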(event EventOnDeathPlayEffects
  (params (Damage :type Float%20%28single-precision%29) (DamageTags :type Gameplay%20Tag%20Container%20Structure%20%28by%20ref%29) (Momentum :type Vector :default "0, 0, 0") (HitInfo :type Hit%20Result%20Structure%20%28by%20ref%29) (InstigatedBy :type Fort%20Pawn%20Object%20Reference) (DamageCauser :type Actor%20Object%20Reference) (EffectContext :type Gameplay%20Effect%20Context%20Handle%20Structure))
  ;[1 1 1] Handle the destruction of things
  (Networking|CanExecuteCosmeticEvents
    (:True
      (Audio|PlaySoundatLocation "/Game/Athena/Sounds/SC_Machine/SC_Machine_Blast_Cue.SC_Machine_Blast_Cue" (Transformation|GetActorLocation (Variables|/Script/Engine.ChildActorComponent|GetChildActor (Variables|Default|GetLocation)))))))

(event CloneMachine|EventOnCloneMachineActivationBegins
  ;[1 1 1] Handle Clone Machine Activation
  (if (< (Fortnite|Sync|Time|GetServerWorldTimeSeconds) (Variables|BuildingGameplayActorCloneMachine|GetActivationServerEndTime))
    (bind _returnvalue (Audio|SpawnSoundatLocation "/Game/Athena/Sounds/SC_Machine/SC_Machine_Activate_Cue.SC_Machine_Activate_Cue" (Transformation|GetActorLocation (Variables|/Script/Engine.ChildActorComponent|GetChildActor (Variables|Default|GetLocation)))))
    (Variables|Default|SetSoundsPlaying _returnvalue)
    (Utilities|Time|SetTimerbyEvent (EventDispatchers|CreateEvent :!event Custom|HandleActivationComplete) (- (Variables|BuildingGameplayActorCloneMachine|GetActivationServerEndTime) (Fortnite|Sync|Time|GetServerWorldTimeSeconds)))))

(event Custom|HandleActivationComplete
  ;[1 1 1] Handle Clone Machine Activation
  (bind _soundsplaying (Variables|Default|GetSoundsPlaying)
    (:then
      (Audio|Components|Audio|FadeOut _soundsplaying 1.0 0.0 "Linear" :!class /Script/Engine.AudioComponent)
      (Audio|PlaySoundatLocation "/Game/Athena/Sounds/SC_Machine/SC_Machine_Card_Spawn_In_Cue.SC_Machine_Card_Spawn_In_Cue" (Transformation|GetActorLocation (Variables|/Script/Engine.ChildActorComponent|GetChildActor (Variables|Default|GetLocation)))))
    (:else)))

(event CloneMachine|EventOnCloneMachineRespawningBegins
  ;[1 1 1] Handle Respawn Beginning
  (if (< (Fortnite|Sync|Time|GetServerWorldTimeSeconds) (Variables|BuildingGameplayActorCloneMachine|GetRespawnServerEndTime))
    (Audio|Components|Audio|FadeOut (Variables|Default|GetAmbientSound) 1.0 0.0 "Linear" :!class /Script/Engine.AudioComponent)
    (bind _returnvalue (Audio|SpawnSoundatLocation "/Game/Athena/Sounds/SC_Machine/SC_Machine_Activate_Cue.SC_Machine_Activate_Cue" (Transformation|GetActorLocation (Variables|/Script/Engine.ChildActorComponent|GetChildActor (Variables|Default|GetLocation)))))
    (Variables|Default|SetSoundsPlaying _returnvalue)
    (Utilities|Time|SetTimerbyEvent (EventDispatchers|CreateEvent :!event Custom|HandleRespawnBeginComplete) (- (Variables|BuildingGameplayActorCloneMachine|GetRespawnServerEndTime) (Fortnite|Sync|Time|GetServerWorldTimeSeconds)))))

(event Custom|HandleRespawnBeginComplete
  ;[1 1 1] Handle Respawn Beginning
  (bind _soundsplaying (Variables|Default|GetSoundsPlaying)
    (:then
      (Audio|Components|Audio|FadeOut _soundsplaying 1.0 0.0 "Linear" :!class /Script/Engine.AudioComponent))
    (:else)))

(event CloneMachine|EventOnCloneMachineSpawnsPlayer
  (params (PlayerPawn :type Fort%20Player%20Pawn%20Object%20Reference))
  ;[1 1 1] TriggerRespawnFX- Hide Mannequin. play Mannequin FX and spawn in the bot.
  (CallFunction|TriggerRespawnFX PlayerPawn))

(event Custom|TriggerRespawnFX
  (params (PhaseInPawn :type Fort%20Pawn%20Object%20Reference))
  ;[1 1 1] TriggerRespawnFX- Hide Mannequin. play Mannequin FX and spawn in the bot.
  (if (Networking|IsServer)
    (bind _returnvalue (Game|SpawnActorfromClass "/Game/Blueprints/Adventures/SpawnTeleporter/BP_VictoryDrone.BP_VictoryDrone_C" (Math|Transform|MakeTransform (Transformation|GetActorLocation (Variables|/Script/Engine.ChildActorComponent|GetChildActor (Variables|Default|GetLocation))) (Transformation|GetActorRotation)) "AlwaysSpawn" "OverrideRootScale" 0 true true 0 true false false 0))
    (Actor|SetLifeSpan _returnvalue 2.25)
    (Transformation|AttachActorToActor _returnvalue PhaseInPawn "None" "SnapToTarget" "SnapToTarget" "KeepRelative" true)
    (Utilities|FlowControl|Delay 0.16)
    (bind _asplayer_pawn_athena (Utilities|Casting|CastToPlayerPawn_Athena PhaseInPawn))
    (Class|PlayerPawnAthena|ActivateInvincibleVisuals _asplayer_pawn_athena 5.0)
    (else
      (Utilities|FlowControl|Delay 0.16)
      (GameplayCue|ExecuteGameplayCue%28Burst%29Local PhaseInPawn (Variables|Default|GetResIn%20GC) (Ability|GameplayCue|MakeGameplayCueParameters 0.0)))))

(fn ConstructionScript)
